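# Physics asset for the MediCrow medium male player body cosmetic, covering simulation of its skirt, back feathers and leg chains.
# Capsules on the pelvis, thighs and calves form the lower body, with tiny base bodies anchoring each dangling piece to the pelvis or thighs through free constraints.
# Dynamic capsule chains hang from those anchors using twist-locked joints with per-segment swing limits, and the back feathers branch left and right from a center segment.
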PhysicsAsset /BRCosmetics/Characters/Player/Male/Medium/Bodies/M_MED_MediCrow/Meshes/M_MED_MediCrow_Physics
bodies {
 base_MC_leg_chain_l [{shapeType=Capsule, length=0.505, radius=0.505, shapeName=base_MC_leg_chain_l_capsule, rotation=<rot 0 -90 0>}]
 base_MC_leg_chain_r [{shapeType=Capsule, length=0.505, radius=0.505, shapeName=base_MC_leg_chain_r_capsule, rotation=<rot 0 -90 0>}]
 base_MC_skirt_back_c [{shapeType=Capsule, length=0.505, radius=0.505, shapeName=base_MC_skirt_back_c_capsule, rotation=<rot 0 -90 0>}]
 base_skirt_MC_side_l [{shapeType=Capsule, length=0.505, radius=0.505, shapeName=base_skirt_MC_side_l_capsule, rotation=<rot 0 -90 0>}]
 base_skirt_MC_side_r [{shapeType=Capsule, length=0.505, radius=0.505, shapeName=base_skirt_MC_side_r_capsule, rotation=<rot 0 -90 0>}]
 calf_l [
  {shapeType=Capsule, length=20, radius=8, center=<-10 0 0>, rotation=<rot 90 0 0>}
  {shapeType=Capsule, length=20, radius=8, center=<-10 0 0>, rotation=<rot 90 83.2786 -6.7213>}
 ]
 calf_r [
  {shapeType=Capsule, length=20, radius=8, center=<10 0 0>, rotation=<rot 90 0 0>}
  {shapeType=Capsule, length=20, radius=8, center=<10 0 0>, rotation=<rot 90 83.2785 -6.7214>}
 ]
 dyn_MC_leg_chain_l [{shapeType=Capsule, length=25, radius=2, center=<15 0 8>, rotation=<rot 90 -85.2198 -85.2205>}]
 dyn_MC_leg_chain_r [{shapeType=Capsule, length=25, radius=2, center=<15 0 8>, rotation=<rot 90 0 0>}]
 dyn_MC_skirt_back_c_01 [{shapeType=Capsule, length=12, radius=4, center=<6 0 0>, rotation=<rot 0 0 90>}]
 dyn_MC_skirt_back_c_02 [{shapeType=Capsule, length=10, radius=4.5, center=<6 0 0>, rotation=<rot 0 0 90>}]
 dyn_MC_skirt_back_c_03 [{shapeType=Capsule, length=8, radius=4, center=<6 0 0>, rotation=<rot 0 0 90>}]
 dyn_MC_skirt_feathers_back_01 [{shapeType=Capsule, length=12, radius=4, center=<6 0 3>, rotation=<rot 0 0 90>}]
 dyn_MC_skirt_feathers_back_02_l [{shapeType=Capsule, length=12, radius=4, center=<6 0 3>, rotation=<rot 90 0 0>}]
 dyn_MC_skirt_feathers_back_02_r [{shapeType=Capsule, length=12, radius=4, center=<6 0 3>, rotation=<rot 90 0 0>}]
 dyn_skirt_MC_side_l_01 [{shapeType=Capsule, length=20, radius=4, center=<6 0 3.5>, rotation=<rot 0 0 90>}]
 dyn_skirt_MC_side_l_02 [{shapeType=Capsule, length=20, radius=5, center=<6 0 3.5>, rotation=<rot 0 0 90>}]
 dyn_skirt_MC_side_r_01 [{shapeType=Capsule, length=20, radius=4, center=<6 0 3.5>, rotation=<rot 0 0 90>}]
 dyn_skirt_MC_side_r_02 [{shapeType=Capsule, length=20, radius=5, center=<6 0 3.5>, rotation=<rot 0 0 90>}]
 pelvis [{shapeType=Capsule, length=15, radius=10}]
 thigh_l [{shapeType=Capsule, length=20, radius=8, center=<-15 0 0>, rotation=<rot 90 0 0>}]
 thigh_r [{shapeType=Capsule, length=20, radius=8, center=<15 0 0>, rotation=<rot 90 0 0>}]
}
constraints [
 {
  bone1Name=dyn_MC_skirt_back_c_01
  bone2Name=base_MC_skirt_back_c
  swing1LimitDegrees=15
  swing2LimitDegrees=20
  twistLimitDegrees=1
  twistMotion=Locked
 }
 {
  bone1Name=dyn_MC_skirt_back_c_02
  bone2Name=dyn_MC_skirt_back_c_01
  swing1LimitDegrees=15
  swing2LimitDegrees=30
  twistLimitDegrees=1
  twistMotion=Locked
 }
 {
  bone1Name=dyn_MC_skirt_back_c_03
  bone2Name=dyn_MC_skirt_back_c_02
  swing1LimitDegrees=35
  swing2LimitDegrees=35
  twistLimitDegrees=1
  twistMotion=Locked
 }
 {
  bone1Name=dyn_skirt_MC_side_l_01
  bone2Name=base_skirt_MC_side_l
  swing1LimitDegrees=30
  swing2LimitDegrees=5
  twistLimitDegrees=1
  twistMotion=Locked
 }
 {
  bone1Name=dyn_skirt_MC_side_l_02
  bone2Name=dyn_skirt_MC_side_l_01
  swing1LimitDegrees=45
  swing2LimitDegrees=25
  twistLimitDegrees=1
  twistMotion=Locked
 }
 {
  bone1Name=dyn_skirt_MC_side_r_01
  bone2Name=base_skirt_MC_side_r
  swing1LimitDegrees=30
  swing2LimitDegrees=5
  twistLimitDegrees=1
  twistMotion=Locked
 }
 {
  bone1Name=dyn_skirt_MC_side_r_02
  bone2Name=dyn_skirt_MC_side_r_01
  swing1LimitDegrees=45
  swing2LimitDegrees=25
  twistLimitDegrees=1
  twistMotion=Locked
 }
 {
  bone1Name=base_skirt_MC_side_l
  bone2Name=thigh_l
  swing1LimitDegrees=45
  swing1Motion=Free
  swing2LimitDegrees=45
  swing2Motion=Free
  twistLimitDegrees=45
  twistMotion=Free
 }
 {
  bone1Name=base_skirt_MC_side_r
  bone2Name=thigh_r
  swing1LimitDegrees=45
  swing1Motion=Free
  swing2LimitDegrees=45
  swing2Motion=Free
  twistLimitDegrees=45
  twistMotion=Free
 }
 {
  bone1Name=base_MC_skirt_back_c
  bone2Name=pelvis
  swing1LimitDegrees=45
  swing1Motion=Free
  swing2LimitDegrees=45
  swing2Motion=Free
  twistLimitDegrees=45
  twistMotion=Free
 }
 {
  bone1Name=thigh_l
  bone2Name=pelvis
  swing1LimitDegrees=45
  swing1Motion=Free
  swing2LimitDegrees=45
  swing2Motion=Free
  twistLimitDegrees=45
  twistMotion=Free
 }
 {
  bone1Name=calf_l
  bone2Name=thigh_l
  swing1LimitDegrees=45
  swing1Motion=Free
  swing2LimitDegrees=45
  swing2Motion=Free
  twistLimitDegrees=45
  twistMotion=Free
 }
 {
  bone1Name=thigh_r
  bone2Name=pelvis
  swing1LimitDegrees=45
  swing1Motion=Free
  swing2LimitDegrees=45
  swing2Motion=Free
  twistLimitDegrees=45
  twistMotion=Free
 }
 {
  bone1Name=calf_r
  bone2Name=thigh_r
  swing1LimitDegrees=45
  swing1Motion=Free
  swing2LimitDegrees=45
  swing2Motion=Free
  twistLimitDegrees=45
  twistMotion=Free
 }
 {
  bone1Name=base_MC_leg_chain_l
  bone2Name=thigh_l
  swing1LimitDegrees=45
  swing1Motion=Free
  swing2LimitDegrees=45
  swing2Motion=Free
  twistLimitDegrees=45
  twistMotion=Free
 }
 {
  bone1Name=dyn_MC_leg_chain_l
  bone2Name=base_MC_leg_chain_l
  swing1LimitDegrees=70
  swing2LimitDegrees=35
  twistLimitDegrees=1
  twistMotion=Locked
 }
 {
  bone1Name=base_MC_leg_chain_r
  bone2Name=thigh_r
  swing1LimitDegrees=45
  swing1Motion=Free
  swing2LimitDegrees=45
  swing2Motion=Free
  twistLimitDegrees=45
  twistMotion=Free
 }
 {
  bone1Name=dyn_MC_leg_chain_r
  bone2Name=base_MC_leg_chain_r
  swing1LimitDegrees=70
  swing2LimitDegrees=35
  twistLimitDegrees=1
  twistMotion=Locked
 }
 {
  bone1Name=dyn_MC_skirt_feathers_back_01
  bone2Name=base_MC_skirt_back_c
  swing1LimitDegrees=25
  swing2LimitDegrees=20
  twistLimitDegrees=45
  twistMotion=Locked
 }
 {
  bone1Name=dyn_MC_skirt_feathers_back_02_l
  bone2Name=dyn_MC_skirt_feathers_back_01
  swing1LimitDegrees=25
  swing2LimitDegrees=20
  twistLimitDegrees=45
  twistMotion=Locked
 }
 {
  bone1Name=dyn_MC_skirt_feathers_back_02_r
  bone2Name=dyn_MC_skirt_feathers_back_01
  swing1LimitDegrees=25
  swing2LimitDegrees=20
  twistLimitDegrees=45
  twistMotion=Locked
 }
]
