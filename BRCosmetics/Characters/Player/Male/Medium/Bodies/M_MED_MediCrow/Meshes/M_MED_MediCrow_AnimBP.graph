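# Secondary-motion AnimBP for the MediCrow cosmetic body, driving feathers, skirt and pouches on a pose copied from the attached parent mesh.
# A deform Control Rig feeds weighted bone constraints, then RigidBody simulation switches between gameplay and front-end physics assets.
# Skirt rotation constraints, AnimDynamics feathers and a skirt bone offset follow.
# The event graph sets movement/emote alphas, disables AnimDynamics on ziplines and reduces gameplay RigidBody alpha under capes or long backblings.
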AnimBlueprint /BRCosmetics/Characters/Player/Male/Medium/Bodies/M_MED_MediCrow/Meshes/M_MED_MediCrow_AnimBP
target_skeleton=&/M_MED_MediCrow_Skeleton
graphs [
 AnimGraph {nodes [
   n1 = RigidBody!(
    ComponentPose=Local To Component(LocalPose=Use cached pose 'surfaceConstraints_cache'())
    Alpha=Get Alpha_RB_WearingCape() -> (Alpha_RB_WearingCape:double)
    OverrideWorldGravity:Vector=Get GravityOverride()
    ExternalForce
   ) -> (Pose) {
    bClampLinearTranslationLimitToRefPose=true
    bFreezeIncomingPoseOnStart=true
    bOverrideWorldGravity=true
    bUseExternalClothCollision=true
    overridePhysicsAsset=&/M_MED_MediCrow_Physics
    simulationSpace=BaseBoneSpace
    alphaBoolBlend {blendOption=Linear}
    baseBoneRef {boneName=pelvis}
    componentAppliedLinearAccClamp=<3000 3000 3000>
    overrideWorldGravity {z=-980}
    simSpaceSettings {maxAngularAcceleration=300, maxAngularVelocity=9, maxLinearAcceleration=1200, maxLinearVelocity=600, velocityScaleZ=5, worldAlpha=1}
   }
   n2 = Component To Local(ComponentPose=@n1)
   n3 = RigidBody(ComponentPose=Local To Component(LocalPose=Use cached pose 'surfaceConstraints_cache'()), OverrideWorldGravity=Get GravityOverride()) {
    bClampLinearTranslationLimitToRefPose=true
    bFreezeIncomingPoseOnStart=true
    bOverrideWorldGravity=true
    bUseExternalClothCollision=true
    overridePhysicsAsset=&/M_MED_MediCrow_FrontEnd_Physics
    simulationSpace=BaseBoneSpace
    alphaBoolBlend {blendOption=Linear}
    baseBoneRef {boneName=pelvis}
    componentAppliedLinearAccClamp=<3000 3000 3000>
    overrideWorldGravity {z=-980}
    simSpaceSettings {maxAngularAcceleration=300, maxAngularVelocity=9, maxLinearAcceleration=1200, maxLinearVelocity=600, velocityScaleZ=5, worldAlpha=1}
   }
   n4 = Component To Local(ComponentPose=@n3)
   n5 = Get bIsInFrontEnd()
   n6 = "Constraint - dyn_MC_skirt_feathers_back_02_l (R)"!(
    ComponentPose=Local To Component(LocalPose=Blend Poses by bool(BlendPose_0=@n4, BlendPose_1=@n2, bActiveValue=@n5) {blendType=Linear})
    bAlphaBoolEnabled:bool=True
    ConstraintWeights_0=1
   ) -> (Pose) {
    alphaInputType=Bool
    alphaBoolBlend {blendInTime=0.2, blendOption=Linear, blendOutTime=0.2}
    boneToModify {boneName=dyn_MC_skirt_feathers_back_02_l}
    constraintSetup [{transformType=Rotation, targetBone {boneName=pelvis}}]
    constraintWeights [1]
   }
   n7 = "Constraint - dyn_MC_skirt_feathers_back_02_r (R)"!(ComponentPose=@n6, bAlphaBoolEnabled:bool=True, ConstraintWeights_0=1) -> (Pose) {
    alphaInputType=Bool
    alphaBoolBlend {blendInTime=0.2, blendOption=Linear, blendOutTime=0.2}
    boneToModify {boneName=dyn_MC_skirt_feathers_back_02_r}
    constraintSetup [{transformType=Rotation, targetBone {boneName=pelvis}}]
    constraintWeights [1]
   }
   n8 = "Constraint - dyn_MC_skirt_back_c_02 (R)"!(ComponentPose=@n7, bAlphaBoolEnabled:bool=True, ConstraintWeights_0=1) -> (Pose) {
    alphaInputType=Bool
    alphaBoolBlend {blendInTime=0.2, blendOption=Linear, blendOutTime=0.2}
    boneToModify {boneName=dyn_MC_skirt_back_c_02}
    constraintSetup [{transformType=Rotation, targetBone {boneName=pelvis}}]
    constraintWeights [1]
   }
   n9 = Get animD alpha() -> (animD alpha:double)
   n10 = Anim Dynamics - Bone: dyn_MC_forearm_feathers_l(ComponentPose=@n8, Alpha=@n9, ExternalForce="0,0,0") {
    angularDampingOverride=1
    angularSpringConstant=20000
    bAngularSpring=true
    bOverrideAngularDamping=true
    bUseGravityOverride=true
    gravityScale=0.35
    simulationSpace=RootRelative
    alphaBoolBlend {blendOption=Linear}
    boundBone {boneName=dyn_MC_forearm_feathers_l}
    componentLinearAccScale=<1 1 1>
    componentLinearVelScale=<1 1 1>
    gravityOverride {z=-980}
    physicsBodyDefinitions [{
      boundBone {boneName=dyn_MC_forearm_feathers_l}
      boxExtents=<35 35 35>
      constraintSetup {angularLimitsMax=<15 15 15>, angularLimitsMin=<-15 -15 -15>, angularTarget=<1 0 0>}
      localJointOffset=<4 0 0>
     }]
    relativeSpaceBone {boneName=lowerarm_l}
   }
   n11 = Anim Dynamics - Bone: dyn_MC_forearm_feathers_r(ComponentPose=@n10, Alpha=@n9, ExternalForce="0,0,0") {
    angularDampingOverride=1
    angularSpringConstant=20000
    bAngularSpring=true
    bOverrideAngularDamping=true
    bUseGravityOverride=true
    gravityScale=0.35
    simulationSpace=BoneRelative
    alphaBoolBlend {blendOption=Linear}
    boundBone {boneName=dyn_MC_forearm_feathers_r}
    componentLinearAccScale=<1 1 1>
    componentLinearVelScale=<1 1 1>
    gravityOverride {z=-980}
    physicsBodyDefinitions [{
      boundBone {boneName=dyn_MC_forearm_feathers_r}
      boxExtents=<40 40 40>
      constraintSetup {angularLimitsMax=<15 15 15>, angularLimitsMin=<-15 -15 -15>, angularTarget=<1 0 0>}
      localJointOffset=<4 0 0>
     }]
    relativeSpaceBone {boneName=lowerarm_r}
   }
   n12 = Anim Dynamics - Bone: dyn_MC_feather_torso_top_l(ComponentPose=@n11, Alpha=@n9, ExternalForce="0,0,0") {
    angularDampingOverride=1
    angularSpringConstant=20000
    bAngularSpring=true
    bOverrideAngularDamping=true
    bUseGravityOverride=true
    gravityScale=0.35
    simulationSpace=BoneRelative
    alphaBoolBlend {blendOption=Linear}
    boundBone {boneName=dyn_MC_feather_torso_top_l}
    componentLinearAccScale=<1 1 1>
    componentLinearVelScale=<1 1 1>
    gravityOverride {z=-980}
    physicsBodyDefinitions [{
      boundBone {boneName=dyn_MC_feather_torso_top_l}
      boxExtents=<40 40 40>
      constraintSetup {angularLimitsMax=<25 25 25>, angularLimitsMin=<-25 -25 -25>, angularTarget=<1 0 0>}
      localJointOffset=<4 0 0>
     }]
    relativeSpaceBone {boneName=spine_04}
   }
   n13 = Anim Dynamics - Bone: dyn_MC_feather_torso_bottom_l(ComponentPose=@n12, Alpha=@n9, ExternalForce="0,0,0") {
    angularDampingOverride=1
    angularSpringConstant=20000
    bAngularSpring=true
    bOverrideAngularDamping=true
    bUseGravityOverride=true
    gravityScale=0.35
    simulationSpace=BoneRelative
    alphaBoolBlend {blendOption=Linear}
    boundBone {boneName=dyn_MC_feather_torso_bottom_l}
    componentLinearAccScale=<1 1 1>
    componentLinearVelScale=<1 1 1>
    gravityOverride {z=-980}
    physicsBodyDefinitions [{
      boundBone {boneName=dyn_MC_feather_torso_bottom_l}
      boxExtents=<35 35 35>
      constraintSetup {angularLimitsMax=<30 0 30>, angularLimitsMin=<0 -30 -30>, angularTarget=<1 0 0>}
      localJointOffset=<4 0 0>
     }]
    relativeSpaceBone {boneName=spine_03}
   }
   n14 = Anim Dynamics - Bone: dyn_MC_feather_torso_top_r(ComponentPose=@n13, Alpha=@n9, ExternalForce="0,0,0") {
    angularDampingOverride=1
    angularSpringConstant=20000
    bAngularSpring=true
    bOverrideAngularDamping=true
    bUseGravityOverride=true
    gravityScale=0.35
    simulationSpace=BoneRelative
    alphaBoolBlend {blendOption=Linear}
    boundBone {boneName=dyn_MC_feather_torso_top_r}
    componentLinearAccScale=<1 1 1>
    componentLinearVelScale=<1 1 1>
    gravityOverride {z=-980}
    physicsBodyDefinitions [{
      boundBone {boneName=dyn_MC_feather_torso_top_r}
      boxExtents=<35 35 35>
      constraintSetup {angularLimitsMax=<25 25 25>, angularLimitsMin=<-25 -25 -25>, angularTarget=<1 0 0>}
      localJointOffset=<4 0 0>
     }]
    relativeSpaceBone {boneName=spine_04}
   }
   n15 = Anim Dynamics - Bone: dyn_MC_feather_torso_bottom_r(ComponentPose=@n14, Alpha=@n9, ExternalForce="0,0,0") {
    angularDampingOverride=1
    angularSpringConstant=20000
    bAngularSpring=true
    bOverrideAngularDamping=true
    bUseGravityOverride=true
    gravityScale=0.35
    simulationSpace=BoneRelative
    alphaBoolBlend {blendOption=Linear}
    boundBone {boneName=dyn_MC_feather_torso_bottom_r}
    componentLinearAccScale=<1 1 1>
    componentLinearVelScale=<1 1 1>
    gravityOverride {z=-980}
    physicsBodyDefinitions [{
      boundBone {boneName=dyn_MC_feather_torso_bottom_r}
      boxExtents=<40 40 40>
      constraintSetup {angularLimitsMax=<0 0 30>, angularLimitsMin=<-30 -30 -30>, angularTarget=<1 0 0>}
      localJointOffset=<4 0 0>
     }]
    relativeSpaceBone {boneName=spine_03}
   }
   n16 = "Transform (Modify) Bone - Bone: base_skirt_MC_side_r"(ComponentPose=@n15, Alpha=1, Translation="0,0,0", Rotation="2.5,0,0", Scale="1,1,1") {
    rotationMode=BMM_Additive
    rotationSpace=BCS_BoneSpace
    translationMode=BMM_Additive
    translationSpace=BCS_BoneSpace
    alphaBoolBlend {blendOption=Linear}
    boneToModify {boneName=base_skirt_MC_side_r}
    rotation {pitch=2.5}
   }
   Output Pose(Result=Component To Local(ComponentPose=@n16)) {layerGroup=None}
   n17 = Control Rig!(
    Source=Copy Pose From Mesh() {bUseAttachedParent=true}
    Alpha
    isSkydiving:double=Get skydiving() -> (skydiving:double)
    isSprinting_float:double=Get isSprintingAlpha() -> (isSprintingAlpha:double)
    isCrouchingMovingFloat:double=Get Is Crouch Walking() -> (Is Crouch Walking:double)
    isCrouching:double=Get isCrouching() -> (isCrouching:double)
    isEmotePlaying:double=Get playingEmote() -> (playingEmote:double)
   ) -> (Pose) {
    alphaBoolBlend {blendOption=Linear}
    controlRigAssetReference {blueprintRigClass=&/M_MED_MediCrow_DeformRig.M_MED_MediCrow_DeformRig_C}
   }
   Save cached pose 'ctrlRig_cache'(Pose=@n17) {cacheName=ctrlRig_cache}
   n18 = "Constraint - base_MC_feather_torso_top_r (PPPPPPP)"!(
    ComponentPose=Local To Component(LocalPose=Use cached pose 'ctrlRig_cache'())
    Alpha=1
    ConstraintWeights_0=0.2537
    ConstraintWeights_1:float=0.239
    ConstraintWeights_2:float=0.2
    ConstraintWeights_3:float=0.1893
    ConstraintWeights_4:float=0.0727
    ConstraintWeights_5:float=0.0234
    ConstraintWeights_6:float=0.0218
   ) -> (Pose) {
    alphaBoolBlend {blendOption=Linear}
    boneToModify {boneName=base_MC_feather_torso_top_r}
    constraintSetup [
     {targetBone {boneName=clavicle_r}}
     {targetBone {boneName=spine_03}}
     {targetBone {boneName=spine_04}}
     {targetBone {boneName=deform_upperarm_bk_r}}
     {targetBone {boneName=spine_02}}
     {targetBone {boneName=upperarm_twist_02_r}}
     {targetBone {boneName=spine_01}}
    ]
    constraintWeights [1,1,1,1,1,1,1]
   }
   n19 = "Constraint - base_MC_feather_torso_bottom_l (PPPPP)"!(
    ComponentPose=@n18
    Alpha=1
    ConstraintWeights_0=0.3706
    ConstraintWeights_1:float=0.3077
    ConstraintWeights_2:float=0.1358
    ConstraintWeights_3:float=0.1005
    ConstraintWeights_4:float=0.0854
   ) -> (Pose) {
    alphaBoolBlend {blendOption=Linear}
    boneToModify {boneName=base_MC_feather_torso_bottom_l}
    constraintSetup [
     {targetBone {boneName=spine_03}}
     {targetBone {boneName=spine_02}}
     {targetBone {boneName=spine_01}}
     {targetBone {boneName=spine_04}}
     {targetBone {boneName=pelvis}}
    ]
    constraintWeights [1,1,1,1,1]
   }
   n20 = "Constraint - base_MC_feather_torso_top_l (PPPPPPP)"!(
    ComponentPose=@n19
    Alpha=1
    ConstraintWeights_0=0.2537
    ConstraintWeights_1:float=0.239
    ConstraintWeights_2:float=0.2
    ConstraintWeights_3:float=0.1893
    ConstraintWeights_4:float=0.0727
    ConstraintWeights_5:float=0.0234
    ConstraintWeights_6:float=0.0218
   ) -> (Pose) {
    alphaBoolBlend {blendOption=Linear}
    boneToModify {boneName=base_MC_feather_torso_top_l}
    constraintSetup [
     {targetBone {boneName=clavicle_l}}
     {targetBone {boneName=spine_03}}
     {targetBone {boneName=spine_04}}
     {targetBone {boneName=deform_upperarm_bk_l}}
     {targetBone {boneName=spine_02}}
     {targetBone {boneName=upperarm_twist_02_l}}
     {targetBone {boneName=spine_01}}
    ]
    constraintWeights [1,1,1,1,1,1,1]
   }
   n21 = "Constraint - base_MC_feather_torso_bottom_r (PPPPP)"!(
    ComponentPose=@n20
    Alpha=1
    ConstraintWeights_0=0.3706
    ConstraintWeights_1:float=0.3077
    ConstraintWeights_2:float=0.1358
    ConstraintWeights_3:float=0.1005
    ConstraintWeights_4:float=0.0854
   ) -> (Pose) {
    alphaBoolBlend {blendOption=Linear}
    boneToModify {boneName=base_MC_feather_torso_bottom_r}
    constraintSetup [
     {targetBone {boneName=spine_03}}
     {targetBone {boneName=spine_02}}
     {targetBone {boneName=spine_01}}
     {targetBone {boneName=spine_04}}
     {targetBone {boneName=pelvis}}
    ]
    constraintWeights [1,1,1,1,1]
   }
   n22 = "Constraint - base_MC_pouch_l (PPPP)"!(
    ComponentPose=@n21
    Alpha=1
    ConstraintWeights_0=0.5132
    ConstraintWeights_1:float=0.2048
    ConstraintWeights_2:float=0.1669
    ConstraintWeights_3:float=0.1152
   ) -> (Pose) {
    alphaBoolBlend {blendOption=Linear}
    boneToModify {boneName=base_MC_pouch_l}
    constraintSetup [{targetBone {boneName=pelvis}}, {targetBone {boneName=spine_02}}, {targetBone {boneName=spine_01}}, {targetBone {boneName=spine_03}}]
    constraintWeights [1,1,1,1]
   }
   n23 = "Constraint - base_MC_pouch_r (PPPP)"!(
    ComponentPose=@n22
    Alpha=1
    ConstraintWeights_0=0.5132
    ConstraintWeights_1:float=0.2048
    ConstraintWeights_2:float=0.1669
    ConstraintWeights_3:float=0.1152
   ) -> (Pose) {
    alphaBoolBlend {blendOption=Linear}
    boneToModify {boneName=base_MC_pouch_r}
    constraintSetup [{targetBone {boneName=pelvis}}, {targetBone {boneName=spine_02}}, {targetBone {boneName=spine_01}}, {targetBone {boneName=spine_03}}]
    constraintWeights [1,1,1,1]
   }
   Save cached pose 'surfaceConstraints_cache'(Pose=Component To Local(ComponentPose=@n23)) {cacheName=surfaceConstraints_cache}
  ]}
 EventGraph {nodes [
   n1 = Event BlueprintUpdateAnimation()
   > Set deltaTime(deltaTime:double=@n1.DeltaTimeX) -> (Output_Get:double)
   > n2 = Sequence!(execute) -> (then_0, then_1, then_2:Exec)
   > n3 = Gravity Override State Function DEPRECATED(CharacterPartAnimInstance=Self-Reference(), GameTimeInSeconds=GetGameTimeInSeconds(), GravityOverrideParams=Get Gravity Override Params())
   > Set GravityOverride(GravityOverride=@n3)
   n4 = AND Boolean(A=Get bIsTacticalSprinting(), B=NOT Boolean(A=Get bIsCrouching()))
   > Set isSprintingAlpha(isSprintingAlpha:double=FInterpTo(Current=Get isSprintingAlpha(), Target="To Float (Boolean)"(InBool=@n4), DeltaTime=Get deltaTime() -> (deltaTime:double), InterpSpeed=10)) -> (Output_Get:double)
   n5 = Get bIsPlayingEmote()
   > Set playingEmote(playingEmote:double=FInterpTo(Current=Get playingEmote(), Target="To Float (Boolean)"(InBool=@n5), DeltaTime=Get deltaTime(), InterpSpeed=10)) -> (Output_Get:double)
   n6 = Get bIsCrouching()
   n7 = Get bIsMoving2D()
   > Set Is Crouch Walking(Is Crouch Walking:double="To Float (Boolean)"(InBool=AND Boolean(A=@n6, B=@n7))) -> (Output_Get:double)
   > Set skydiving(skydiving:double="To Float (Boolean)"(InBool=Get bIsSkyDiving())) -> (Output_Get:double)
   > Set isCrouching(isCrouching:double="To Float (Boolean)"(InBool=Get bIsCrouching())) -> (Output_Get:double)
   > n8 = Branch(Condition=Get OnZipline() -> (OnZipline:bool))
   > Set animD alpha(animD alpha:double=0) -> (Output_Get:double)
   Set animD alpha(execute=@n8.else, animD alpha=1)
   n9 = CheckForBackbling()
   Set isWearingBackbling(
    execute=@n2.then_1
    isWearingBackbling=AND Boolean!(A=@n9, B=NOT Boolean(A=@n9.bIsBackblingACape), C:bool=NOT Boolean(A=@n9.bIsBackblingLong)) -> (ReturnValue)
   )
   > n10 = Set Is Wearing Cape(Is Wearing Cape=OR Boolean(A=@n9.bIsBackblingACape, B=@n9.bIsBackblingLong))
   > Set Is Wearing CapeCrouch(Is Wearing CapeCrouch:bool=AND Boolean!(A=Get bIsCrouching(), B=@n10) -> (ReturnValue)) -> (Output_Get:bool)
   n11 = Branch(execute=@n2.then_2, Condition=Get Is Wearing Cape() -> (Is Wearing Cape:bool))
   > Set Alpha_RB_WearingCape(Alpha_RB_WearingCape:double=0.6) -> (Output_Get:double)
   Set Alpha_RB_WearingCape(execute=@n11.else, Alpha_RB_WearingCape=1)
   n12 = Event BlueprintProcessAnimInputEvent()
   > n13 = Branch(Condition=IsNamed(self=@n12.AnimInputEvent, InEventName=aie.Player.Zipline.Entered))
   > Set OnZipline(OnZipline:bool=1) -> (Output_Get:bool)
   Branch(execute=@n13.else, Condition=IsNamed(self=@n12.AnimInputEvent, InEventName=aie.Player.Zipline.Exited))
   > Set OnZipline(OnZipline=false)
   Event OnBeginSkydiving()
   Get bIsSprinting()
  ]}
]
variables [
 Alpha_RB_WearingCape {type=number}
 animD alpha {type=number}
 deltaTime {type=number}
 Gravity Override Params {type=GravityOverrideParamsStruct}
 GravityOverride {type=Vector}
 Is Crouch Walking {type=number}
 Is Wearing Cape {type=boolean}
 Is Wearing CapeCrouch {type=boolean}
 isCrouching {type=number}
 isSprintingAlpha {type=number}
 isWearingBackbling {type=boolean}
 OnZipline {type=boolean}
 playingEmote {type=number}
 skydiving {type=number}
]
