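# Animation Blueprint for the MediCrow face accessory, copying the attached parent mesh's pose and damping its head and neck rotation outside emotes.
# Each update negates and scales the head, neck_02 and neck_01 deltas from reference pose, with roll weighted less than pitch and yaw.
# Chained additive bone-space Modify Bone nodes apply these counter-rotations.
# While bIsPlayingEmote is true, a bool blend outputs the unmodified copied pose instead.
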AnimBlueprint /BRCosmetics/Characters/Player/Male/Medium/Bodies/M_MED_MediCrow/Meshes/Parts/M_MED_MediCrow_FaceAcc_AnimBP
target_skeleton=&/M_MED_MediCrow_FaceAcc_Skeleton
graphs [
 AnimGraph {nodes [
   n1 = "Transform (Modify) Bone - Bone: head"(
    ComponentPose=Local To Component(LocalPose=Copy Pose From Mesh() {bUseAttachedParent=true})
    Alpha=1
    Translation="0,0,0"
    Rotation=Get head rot() -> (head rot:Rotator)
    Scale="1,1,1"
   ) {
    rotationMode=BMM_Additive
    rotationSpace=BCS_BoneSpace
    translationMode=BMM_Additive
    translationSpace=BCS_BoneSpace
    alphaBoolBlend {blendOption=Linear}
    boneToModify {boneName=head}
   }
   n2 = "Transform (Modify) Bone - Bone: neck_02"(ComponentPose=@n1, Alpha=1, Translation="0,0,0", Rotation=Get neck02_rot() -> (neck02_rot:Rotator), Scale="1,1,1") {
    rotationMode=BMM_Additive
    rotationSpace=BCS_BoneSpace
    translationMode=BMM_Additive
    translationSpace=BCS_BoneSpace
    alphaBoolBlend {blendOption=Linear}
    boneToModify {boneName=neck_02}
   }
   n3 = "Transform (Modify) Bone - Bone: neck_01"(ComponentPose=@n2, Alpha=1, Translation="0,0,0", Rotation=Get neck01_rot() -> (neck01_rot:Rotator), Scale="1,1,1") {
    rotationMode=BMM_Additive
    rotationSpace=BCS_BoneSpace
    translationMode=BMM_Additive
    translationSpace=BCS_BoneSpace
    alphaBoolBlend {blendOption=Linear}
    boneToModify {boneName=neck_01}
   }
   Output Pose(Result=Blend Poses by bool(
     BlendPose_0=Copy Pose From Mesh() {bUseAttachedParent=true}
     BlendPose_1=Component To Local(ComponentPose=@n3)
     bActiveValue=Get bIsPlayingEmote()
    ) {blendType=Linear}) {layerGroup=None}
  ]}
 EventGraph {nodes [
   Event BlueprintUpdateAnimation()
   > Sequence()
   n1 = GetOwnerSkeletalMeshComponent()
   n2 = BreakRotator(InRot=BreakTransform(InTransform=GetDeltaTransformFromRefPose(self=@n1, BoneName=head, BaseName=neck_02)).Rotation)
   > Set head rot(head rot:Rotator=MakeRotator(
     Roll=float * float!(A=@n2, B=-1, C:double=0.2) -> (ReturnValue)
     Pitch=float * float(A=@n2.Pitch, B=-1, C=0.6)
     Yaw=float * float(A=@n2.Yaw, B=-1, C=0.6)
    )) -> (Output_Get:Rotator)
   n3 = BreakRotator(InRot=BreakTransform(InTransform=GetDeltaTransformFromRefPose(self=@n1, BoneName=neck_02, BaseName=neck_01)).Rotation)
   > Set neck02_rot(neck02_rot:Rotator=MakeRotator(Roll=float * float(A=@n3, B=-1, C=0.2), Pitch=float * float(A=@n3.Pitch, B=-1, C=0.6), Yaw=float * float(A=@n3.Yaw, B=-1, C=0.6))) -> (Output_Get:Rotator)
   n4 = BreakRotator(InRot=BreakTransform(InTransform=GetDeltaTransformFromRefPose(self=@n1, BoneName=neck_01, BaseName=spine_05)).Rotation)
   > Set neck01_rot(neck01_rot:Rotator=MakeRotator(Roll=float * float(A=@n4, B=-1, C=0.2), Pitch=float * float(A=@n4.Pitch, B=-1, C=0.6), Yaw=float * float(A=@n4.Yaw, B=-1, C=0.6))) -> (Output_Get:Rotator)
  ]}
]
variables [
 head rot {type=Rotator}
 neck01_rot {type=Rotator}
 neck02_rot {type=Rotator}
 X percentage {type=number}
 Y percentage {type=number}
 Z percentage {type=number}
]
